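# Authoring Level Sequence for a menu animation of the Season 41 ChewyStaveCurve_M skin, built on a FortPlayerMannequin.
# The OnFace control rig track keys blinks, eye darts, brows, smile, lips and jaw, returning to roughly the opening pose at frame 266.
# MR_Biped body rig sections and a Head-component bone-tweaker rig section show no channel data.
# The skeletal animation track is empty, so the visible motion is all facial.
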LevelSequence /BRCosmetics/Animation/Game/MainPlayer/Menu/BR/CustomSkins/_S41/ChewyStaveCurve_M/ChewyStaveCurve_Anim/ChewyStaveGerm_anim
bindings [
 FortPlayerMannequin {
  id=#1
  possessed_class=FortPlayerMannequin
  tracks [
   {type=MovieSceneSpawnTrack, sections [{type=MovieSceneSpawnSection, channels [None {type=MovieSceneScriptingBoolChannel, default=true}]}]}
   {
    type=MovieScene3DTransformTrack
    sections [{
      type=MovieScene3DTransformSection
      channels [
       Location.X {type=MovieSceneScriptingDoubleChannel, default=0}
       Location.Y {type=MovieSceneScriptingDoubleChannel, default=0}
       Location.Z {type=MovieSceneScriptingDoubleChannel, default=0}
       Rotation.X {type=MovieSceneScriptingDoubleChannel, default=0}
       Rotation.Y {type=MovieSceneScriptingDoubleChannel, default=0}
       Rotation.Z {type=MovieSceneScriptingDoubleChannel, default=0}
       Scale.X {type=MovieSceneScriptingDoubleChannel, default=1}
       Scale.Y {type=MovieSceneScriptingDoubleChannel, default=1}
       Scale.Z {type=MovieSceneScriptingDoubleChannel, default=1}
      ]
     }]
   }
   {type=MovieSceneSkeletalAnimationTrack, sections []}
   {
    type=MovieSceneControlRigParameterTrack
    display_name=Fortnite_3L_OnFace_CtrlRig
    sections [{
      type=MovieSceneControlRigParameterSection
      channels [
       CTRL_C_eyesAim.Location.X {type=MovieSceneScriptingFloatChannel, default=0, const [0,28]}
       CTRL_C_eyesAim.Location.Y {type=MovieSceneScriptingFloatChannel, default=0, const [0,28]}
       CTRL_C_eyesAim.Location.Z {type=MovieSceneScriptingFloatChannel, default=0, const [0,28]}
       CTRL_C_eyesAim.Rotation.X {type=MovieSceneScriptingFloatChannel, default=0, const [0,28]}
       CTRL_C_eyesAim.Rotation.Y {type=MovieSceneScriptingFloatChannel, default=0, const [0,28]}
       CTRL_C_eyesAim.Rotation.Z {type=MovieSceneScriptingFloatChannel, default=0, const [0,28]}
       CTRL_C_eyesAim.Scale.X {type=MovieSceneScriptingFloatChannel, default=1, const [1,28]}
       CTRL_C_eyesAim.Scale.Y {type=MovieSceneScriptingFloatChannel, default=1, const [1,28]}
       CTRL_C_eyesAim.Scale.Z {type=MovieSceneScriptingFloatChannel, default=1, const [1,28]}
       CTRL_L_eyeAim.Location.X {type=MovieSceneScriptingFloatChannel, default=0, const [0,28]}
       CTRL_L_eyeAim.Location.Y {type=MovieSceneScriptingFloatChannel, default=0, const [0,28]}
       CTRL_L_eyeAim.Location.Z {type=MovieSceneScriptingFloatChannel, default=0, const [0,28]}
       CTRL_L_eyeAim.Rotation.X {type=MovieSceneScriptingFloatChannel, default=0, const [0,28]}
       CTRL_L_eyeAim.Rotation.Y {type=MovieSceneScriptingFloatChannel, default=0, const [0,28]}
       CTRL_L_eyeAim.Rotation.Z {type=MovieSceneScriptingFloatChannel, default=0, const [0,28]}
       CTRL_L_eyeAim.Scale.X {type=MovieSceneScriptingFloatChannel, default=1, const [1,28]}
       CTRL_L_eyeAim.Scale.Y {type=MovieSceneScriptingFloatChannel, default=1, const [1,28]}
       CTRL_L_eyeAim.Scale.Z {type=MovieSceneScriptingFloatChannel, default=1, const [1,28]}
       CTRL_R_eyeAim.Location.X {type=MovieSceneScriptingFloatChannel, default=0, const [0,28]}
       CTRL_R_eyeAim.Location.Y {type=MovieSceneScriptingFloatChannel, default=0, const [0,28]}
       CTRL_R_eyeAim.Location.Z {type=MovieSceneScriptingFloatChannel, default=0, const [0,28]}
       CTRL_R_eyeAim.Rotation.X {type=MovieSceneScriptingFloatChannel, default=0, const [0,28]}
       CTRL_R_eyeAim.Rotation.Y {type=MovieSceneScriptingFloatChannel, default=0, const [0,28]}
       CTRL_R_eyeAim.Rotation.Z {type=MovieSceneScriptingFloatChannel, default=0, const [0,28]}
       CTRL_R_eyeAim.Scale.X {type=MovieSceneScriptingFloatChannel, default=1, const [1,28]}
       CTRL_R_eyeAim.Scale.Y {type=MovieSceneScriptingFloatChannel, default=1, const [1,28]}
       CTRL_R_eyeAim.Scale.Z {type=MovieSceneScriptingFloatChannel, default=1, const [1,28]}
       CTRL_convergenceSwitch {type=MovieSceneScriptingFloatChannel, default=0, const [0,28]}
       follow_head {type=MovieSceneScriptingFloatChannel, default=0, const [0,28]}
       Face_Ctrl_Master.Location.X {type=MovieSceneScriptingFloatChannel, default=0, const [21.8274,1]}
       Face_Ctrl_Master.Location.Y {type=MovieSceneScriptingFloatChannel, default=0, const [0,28]}
       Face_Ctrl_Master.Location.Z {type=MovieSceneScriptingFloatChannel, default=0, const [0,28]}
       Face_Ctrl_Master.Rotation.X {type=MovieSceneScriptingFloatChannel, default=0, const [0,28]}
       Face_Ctrl_Master.Rotation.Y {type=MovieSceneScriptingFloatChannel, default=0, const [0,28]}
       Face_Ctrl_Master.Rotation.Z {type=MovieSceneScriptingFloatChannel, default=0, const [0,28]}
       Face_Ctrl_Master.Scale.X {type=MovieSceneScriptingFloatChannel, default=1, const [1,28]}
       Face_Ctrl_Master.Scale.Y {type=MovieSceneScriptingFloatChannel, default=1, const [1,28]}
       Face_Ctrl_Master.Scale.Z {type=MovieSceneScriptingFloatChannel, default=1, const [1,28]}
       gui_follow_head {type=MovieSceneScriptingFloatChannel, default=1, const [1,28]}
       CTRL_L_eye_blink {
        type=MovieSceneScriptingFloatChannel
        default=0
        thinned_from=51
        keys [[0,0],[8,0],[11,0.0379],[13,0.3716],[14,0.8913],[15,0.941],[20,1],[21,0.7012],[24,0.273],[28,0.0644],[31,0],[51,0],[54,0.0681],[56,0.8538],[58,1.0369],[59,0.6509],[61,0.0986],[63,0.0384],[67,0],[84,0],[89,-0.0143],[95,-1.4325],[101,-1.6778],[114,-1.7858],[130,-1.8189],[138,-1.8529],[148,-1.9143],[157,-1.9636],[165,-1.9739],[178,-1.8821],[199,-0.7668],[200,0.3661],[201,1],[204,1],[205,0.799],[208,0.1723],[211,0],[235,0],[236,0.0774],[237,0.392],[238,0.9308],[239,3.5869],[240,0.4155],[241,0.1963],[243,0.058],[245,0],[266,0]]
       }
       CTRL_L_eye_eyelidU.X {type=MovieSceneScriptingFloatChannel, default=0, const [0,28]}
       CTRL_L_eye_eyelidU.Y {
        type=MovieSceneScriptingFloatChannel
        default=0
        thinned_from=28
        keys [[0,0],[67,0],[75,0.4559],[84,1],[89,0.7789],[95,0.8328],[101,1],[114,0.9841],[130,0.9572],[138,0.9594],[157,0.9976],[165,1],[178,0.9642],[198,0.8228],[210,0.7093],[235,0.6522],[252,0.0439],[266,0]]
       }
       CTRL_L_eye_eyelidD.X {type=MovieSceneScriptingFloatChannel, default=0, const [0,28]}
       CTRL_L_eye_eyelidD.Y {type=MovieSceneScriptingFloatChannel, default=0, thinned_from=28, keys [[0,0],[51,0.0001],[67,0.0003],[84,0],[266,0]]}
       CTRL_L_brow.X {
        type=MovieSceneScriptingFloatChannel
        default=0
        thinned_from=38
        keys [[0,0],[41,0],[46,0.0001],[51,0.0009],[56,0.0028],[60,0.0038],[75,0.004],[80,0.0126],[84,0.0559],[90,0.1332],[96,0.0557],[102,0.012],[114,0.0033],[130,0.0016],[138,0.0011],[165,0.0006],[181,0.0033],[197,0.041],[200,0.1976],[203,0.3432],[206,0.1585],[210,0.0279],[223,0.0017],[229,0.0003],[236,0.0038],[239,0.0128],[242,0.0077],[245,0.0016],[266,0]]
       }
       CTRL_L_brow.Y {
        type=MovieSceneScriptingFloatChannel
        default=0
        keys [[0,-0.0367],[4,-0.0205],[8,0.0308],[12,0.098],[15,0.1389],[21,0.1602],[27,0.1685],[41,0.1689],[46,0.1512],[51,0.0781],[56,-0.064],[60,-0.1492],[67,-0.1697],[75,-0.1574],[80,-0.1303],[84,-0.0843],[90,0.0226],[96,0.1398],[102,0.1883],[114,0.2],[130,0.2032],[138,0.2027],[148,0.1957],[157,0.1833],[165,0.14],[181,0.1138],[197,0.0909],[200,0.0452],[203,-0.1098],[206,-0.0063],[210,0.0222],[223,0.0372],[229,0.0233],[236,-0.0263],[239,-0.1327],[242,-0.0993],[245,-0.0475],[266,-0.0367]]
       }
       CTRL_L_brow_inner {
        type=MovieSceneScriptingFloatChannel
        default=0
        keys [[0,-0.0367],[4,-0.0395],[8,-0.0497],[15,-0.067],[27,-0.0738],[41,-0.075],[46,-0.0746],[51,-0.071],[56,-0.0617],[60,-0.0561],[67,-0.0521],[75,-0.0282],[84,0.0553],[90,0.1546],[96,0.2052],[102,0.1934],[114,0.1815],[130,0.177],[138,0.1722],[148,0.1588],[157,0.1434],[165,0.1357],[181,0.1292],[197,0.1131],[200,0.0606],[203,0.0223],[206,0.0383],[210,0.0493],[229,0.0401],[236,0.0071],[239,-0.0604],[242,-0.0578],[245,-0.0386],[266,-0.0367]]
       }
       CTRL_L_brow_outer {
        type=MovieSceneScriptingFloatChannel
        default=0
        keys [[0,-0.0367],[4,-0.021],[8,0.0373],[15,0.1405],[27,0.194],[41,0.2045],[46,0.1907],[51,0.1175],[56,-0.0259],[60,-0.106],[67,-0.1225],[75,-0.1002],[84,-0.045],[90,0.0501],[96,0.2653],[102,0.3748],[114,0.4153],[130,0.4226],[138,0.4163],[148,0.3909],[157,0.3567],[165,0.3383],[181,0.3188],[197,0.2845],[200,0.2186],[203,0.0699],[206,-0.1037],[210,-0.1654],[229,-0.1688],[236,-0.1795],[239,-0.2461],[242,-0.152],[245,-0.065],[266,-0.0367]]
       }
       CTRL_C_nose.X {type=MovieSceneScriptingFloatChannel, default=0, const [0,28]}
       CTRL_C_nose.Y {type=MovieSceneScriptingFloatChannel, default=0, const [0,28]}
       CTRL_L_nose.X {type=MovieSceneScriptingFloatChannel, default=0, const [0,28]}
       CTRL_L_nose.Y {
        type=MovieSceneScriptingFloatChannel
        default=0
        thinned_from=28
        keys [[0,-0.092],[75,-0.092],[84,-0.0843],[89,-0.0634],[95,-0.0044],[101,0.0216],[114,0.0386],[130,0.0417],[138,0.0335],[148,0.0056],[157,-0.0773],[165,-0.0891],[178,-0.092],[266,-0.092]]
       }
       noseWrinkleUpper {type=MovieSceneScriptingFloatChannel, default=0, const [0,28]}
       CTRL_R_nose.X {type=MovieSceneScriptingFloatChannel, default=0, const [0,28]}
       CTRL_R_nose.Y {
        type=MovieSceneScriptingFloatChannel
        default=0
        thinned_from=28
        keys [[0,-0.092],[75,-0.092],[84,-0.0843],[89,-0.0634],[95,-0.0044],[101,0.0216],[114,0.0386],[130,0.0417],[138,0.0335],[148,0.0056],[157,-0.0773],[165,-0.0891],[178,-0.092],[266,-0.092]]
       }
       noseWrinkleUpper_2 {type=MovieSceneScriptingFloatChannel, default=0, const [0,28]}
       CTRL_L_eye_squintInner {
        type=MovieSceneScriptingFloatChannel
        default=0
        thinned_from=28
        keys [[0,0],[67,0],[75,0.0431],[84,0.1887],[89,0.2477],[95,0.7235],[101,0.8759],[114,0.8553],[130,0.8204],[138,0.7694],[148,0.5835],[157,0.3034],[165,0.1534],[180,0.0458],[198,0.0468],[210,0.0436],[235,0.0351],[252,0.0076],[266,0]]
       }
       CTRL_L_eye_cheekRaise {
        type=MovieSceneScriptingFloatChannel
        default=0
        thinned_from=28
        keys [[0,0.5521],[67,0.5527],[75,0.5586],[84,0.6064],[89,0.6756],[95,0.7789],[101,0.825],[114,0.85],[130,0.8543],[148,0.8505],[165,0.8514],[178,0.8487],[198,0.8413],[210,0.8325],[235,0.7796],[252,0.578],[266,0.5521]]
       }
       CTRL_C_mouth.X {
        type=MovieSceneScriptingFloatChannel
        default=0
        thinned_from=30
        keys [[0,0.1463],[8,0.1286],[15,0.0756],[23,-0.0168],[31,-0.074],[41,-0.0896],[46,-0.0846],[51,-0.0605],[56,-0.0228],[60,0.0063],[67,0.0288],[75,0.0502],[84,0.0668],[89,0.0755],[95,0.078],[106,0.0802],[114,0.0821],[130,0.087],[138,0.0968],[148,0.1108],[157,0.1268],[165,0.1307],[180,0.2196],[198,0.2276],[210,0.1165],[223,0.1128],[235,0.1097],[252,0.1462],[266,0.1463]]
       }
       CTRL_C_mouth.Y {
        type=MovieSceneScriptingFloatChannel
        default=0
        thinned_from=30
        keys [[0,-0.1957],[31,-0.1956],[46,-0.1962],[51,-0.1969],[56,-0.1981],[67,-0.204],[75,-0.2162],[84,-0.2527],[89,-0.2903],[95,-0.3523],[101,-0.3926],[106,-0.4128],[114,-0.411],[130,-0.3833],[138,-0.3442],[148,-0.3093],[157,-0.3339],[165,-0.4369],[180,-0.6244],[198,-0.6691],[210,-0.4143],[223,-0.3462],[235,-0.2942],[252,-0.1934],[266,-0.1957]]
       }
       CTRL_L_mouth_lipU.X {
        type=MovieSceneScriptingFloatChannel
        default=0
        thinned_from=30
        keys [[0,0],[41,0],[46,-0.0003],[51,-0.001],[56,-0.0023],[60,-0.003],[67,-0.0028],[75,-0.0018],[82,-0.0006],[87,-0.0001],[93,0],[99,-0.0005],[106,-0.0024],[112,-0.01],[130,-0.0315],[138,-0.0666],[142,-0.0981],[146,-0.1102],[155,-0.0495],[163,-0.0192],[180,-0.0197],[198,-0.0197],[210,-0.004],[235,-0.0002],[266,0]]
       }
       CTRL_L_mouth_lipU.Y {
        type=MovieSceneScriptingFloatChannel
        default=0
        thinned_from=30
        keys [[0,-0.1471],[31,-0.1471],[41,-0.1468],[46,-0.1448],[51,-0.1376],[56,-0.126],[60,-0.1185],[67,-0.1136],[75,-0.1039],[82,-0.0821],[87,-0.043],[93,0.0044],[99,0.0256],[106,0.0365],[112,0.0448],[130,0.0712],[138,0.0998],[142,0.0219],[146,-0.149],[155,-0.1309],[163,-0.0642],[180,-0.1631],[198,-0.1715],[210,-0.1037],[235,-0.1132],[252,-0.1344],[266,-0.1471]]
       }
       mouthPress {
        type=MovieSceneScriptingFloatChannel
        default=0
        thinned_from=30
        keys [[0,0],[93,0],[99,0.0007],[106,0.0021],[112,0.0065],[130,0.018],[138,0.04],[142,0.0719],[146,0.1102],[155,0.1531],[163,0.1816],[180,0.1876],[198,0.1879],[210,0.183],[235,0.1364],[252,0.0572],[266,0]]
       }
       lipsTogether {type=MovieSceneScriptingFloatChannel, default=0, thinned_from=30, keys [[0,0],[106,-0.0001],[138,-0.0005],[142,-0.0001],[266,0]]}
       CTRL_L_mouth_lipD.X {
        type=MovieSceneScriptingFloatChannel
        default=0
        thinned_from=32
        keys [[0,-0.0582],[41,-0.0582],[46,-0.0584],[51,-0.0591],[56,-0.0605],[60,-0.0623],[67,-0.07],[75,-0.0949],[82,-0.1396],[87,-0.2005],[93,-0.2815],[99,-0.3132],[106,-0.3267],[112,-0.3289],[116,-0.3273],[121,-0.3264],[130,-0.3359],[138,-0.3628],[142,-0.3806],[146,-0.3408],[155,-0.2287],[163,-0.1636],[180,-0.1589],[198,-0.1579],[210,-0.1263],[235,-0.1024],[252,-0.0759],[266,-0.0582]]
       }
       CTRL_L_mouth_lipD.Y {
        type=MovieSceneScriptingFloatChannel
        default=0
        thinned_from=32
        keys [[0,-0.0637],[31,-0.0637],[41,-0.0641],[46,-0.0662],[51,-0.0734],[56,-0.0847],[60,-0.0896],[67,-0.0768],[75,-0.0271],[82,0.0591],[87,0.161],[93,0.2837],[99,0.3203],[106,0.3302],[112,0.3102],[116,0.2719],[121,0.2445],[130,0.2631],[138,0.27],[142,0.2115],[146,-0.1038],[155,-0.4058],[163,-0.5139],[180,-0.4389],[198,-0.4133],[210,-0.3757],[235,-0.261],[252,-0.1405],[266,-0.0637]]
       }
       mouthPress_2 {
        type=MovieSceneScriptingFloatChannel
        default=0
        thinned_from=31
        keys [[0,0],[99,0.0001],[106,0.0002],[112,0.0006],[116,0.0024],[130,0.0079],[138,0.0188],[142,0.0337],[146,0.061],[155,0.0984],[163,0.1189],[180,0.1211],[198,0.1211],[210,0.116],[235,0.0855],[252,0.035],[266,0]]
       }
       lipsTogether_2 {type=MovieSceneScriptingFloatChannel, default=0, const [0,31]}
       CTRL_L_mouth_smileDepress.X {
        type=MovieSceneScriptingFloatChannel
        default=0
        thinned_from=29
        keys [[0,0.1128],[31,0.1128],[41,0.113],[46,0.1143],[51,0.1186],[56,0.1255],[60,0.1299],[75,0.1338],[82,0.1385],[87,0.1517],[93,0.1754],[99,0.1907],[112,0.1906],[130,0.1822],[138,0.1737],[142,0.1646],[146,0.1533],[155,0.1413],[163,0.1299],[180,0.1013],[198,0.0954],[210,0.0532],[235,0.0463],[252,0.0868],[266,0.1128]]
       }
       CTRL_L_mouth_smileDepress.Y {
        type=MovieSceneScriptingFloatChannel
        default=0
        thinned_from=29
        keys [[0,0.3867],[41,0.3865],[46,0.3851],[51,0.3805],[56,0.3733],[60,0.3714],[67,0.388],[75,0.4478],[82,0.5531],[87,0.6852],[93,0.8511],[99,0.9116],[112,0.9545],[130,0.9108],[138,0.8703],[142,0.8341],[146,0.793],[155,0.7539],[163,0.7328],[180,0.8088],[198,0.8152],[210,0.5284],[235,0.4112],[252,0.399],[266,0.3867]]
       }
       CTRL_L_mouth_corner.X {
        type=MovieSceneScriptingFloatChannel
        default=0
        thinned_from=29
        keys [[0,0],[46,0.0001],[51,0.0005],[60,0.0019],[67,0.0041],[75,0.011],[82,0.0236],[87,0.041],[93,0.0644],[99,0.0785],[112,0.0779],[130,0.0694],[138,0.0608],[142,0.0517],[146,0.0406],[155,0.0309],[163,0.0302],[180,0.0312],[198,0.0313],[210,0.021],[235,0.013],[252,0.0052],[266,0]]
       }
       CTRL_L_mouth_corner.Y {
        type=MovieSceneScriptingFloatChannel
        default=0
        thinned_from=29
        keys [[0,0.7786],[41,0.7788],[46,0.7802],[51,0.7854],[60,0.7995],[67,0.8048],[75,0.8173],[82,0.841],[87,0.8807],[93,0.86],[99,0.5699],[112,0.4846],[130,0.441],[138,0.3131],[142,0.2177],[146,0.1663],[155,0.1268],[163,0.1148],[180,0.124],[198,0.1263],[210,0.1422],[235,0.2997],[252,0.5827],[266,0.7786]]
       }
       lipsBlow {type=MovieSceneScriptingFloatChannel, default=0, const [0,29]}
       mouthDimple {
        type=MovieSceneScriptingFloatChannel
        default=0
        thinned_from=29
        keys [[0,0],[99,0.0001],[112,0.0018],[130,0.0144],[138,0.0624],[142,0.1711],[146,0.3362],[155,0.5007],[163,0.5672],[180,0.5795],[198,0.5802],[210,0.5588],[235,0.4171],[252,0.1711],[266,0]]
       }
       mouthStretch {type=MovieSceneScriptingFloatChannel, default=0, const [0,29]}
       mouthPurseU {type=MovieSceneScriptingFloatChannel, default=0, const [0,29]}
       mouthPurseD {
        type=MovieSceneScriptingFloatChannel
        default=0
        thinned_from=29
        keys [[0,0],[56,0],[60,0.0008],[67,0.006],[75,0.0236],[82,0.055],[87,0.0969],[93,0.1517],[99,0.1844],[112,0.1826],[130,0.1616],[138,0.1372],[142,0.106],[146,0.0646],[155,0.0247],[163,0.0056],[180,0.0038],[198,0.0034],[210,0.001],[266,0]]
       }
       mouthFunnelU {
        type=MovieSceneScriptingFloatChannel
        default=0
        thinned_from=29
        keys [[0,0],[99,0],[112,0.0009],[130,0.0076],[138,0.033],[142,0.0905],[146,0.1778],[155,0.2648],[163,0.3062],[180,0.3097],[198,0.3098],[210,0.2966],[235,0.2205],[252,0.0905],[266,0]]
       }
       mouthFunnelD {type=MovieSceneScriptingFloatChannel, default=0, const [0,29]}
       mouthCornerSharpnessU {
        type=MovieSceneScriptingFloatChannel
        default=0
        thinned_from=29
        keys [[0,0],[56,-0.0001],[60,-0.0015],[67,-0.0117],[75,-0.046],[82,-0.1073],[87,-0.1888],[93,-0.2933],[99,-0.3423],[112,-0.3009],[130,-0.2221],[138,-0.1756],[142,-0.1533],[146,-0.1343],[155,-0.1165],[163,-0.1046],[180,-0.1016],[198,-0.101],[210,-0.0849],[235,-0.0603],[252,-0.0246],[266,0]]
       }
       mouthCornerSharpnessD {
        type=MovieSceneScriptingFloatChannel
        default=0
        thinned_from=29
        keys [[0,0],[56,-0.0001],[60,-0.0012],[67,-0.0092],[75,-0.0357],[82,-0.0818],[87,-0.1401],[93,-0.2263],[99,-0.3291],[112,-0.4459],[130,-0.5281],[138,-0.4847],[142,-0.4405],[146,-0.305],[155,-0.1684],[163,-0.0982],[180,-0.0902],[198,-0.0885],[210,-0.0677],[235,-0.0471],[252,-0.0192],[266,0]]
       }
       CTRL_R_mouth_lipU.X {
        type=MovieSceneScriptingFloatChannel
        default=0
        thinned_from=30
        keys [[0,0],[41,0],[46,-0.0003],[51,-0.001],[56,-0.0023],[60,-0.003],[67,-0.0028],[75,-0.0018],[82,-0.0006],[87,-0.0001],[93,-0.0001],[99,-0.0028],[106,-0.0083],[112,-0.0223],[130,-0.0525],[138,-0.0928],[142,-0.1241],[146,-0.1332],[155,-0.0674],[163,-0.0258],[180,-0.0209],[198,-0.0199],[210,-0.004],[235,-0.0002],[266,0]]
       }
       CTRL_R_mouth_lipU.Y {
        type=MovieSceneScriptingFloatChannel
        default=0
        thinned_from=30
        keys [[0,0],[31,0],[41,0.0003],[46,0.0024],[51,0.0095],[56,0.0211],[60,0.0286],[67,0.0334],[75,0.0426],[82,0.0623],[87,0.0983],[93,0.1463],[99,0.1703],[106,0.182],[112,0.1881],[130,0.208],[138,0.2293],[142,0.1508],[146,-0.015],[155,0.0199],[163,0.0859],[180,0.0643],[198,0.0594],[210,0.0532],[235,0.0318],[252,0.0123],[266,0]]
       }
       mouthPress_3 {
        type=MovieSceneScriptingFloatChannel
        default=0
        thinned_from=30
        keys [[0,0],[93,0],[99,0.001],[106,0.003],[112,0.0082],[130,0.0209],[138,0.0449],[142,0.081],[146,0.1299],[155,0.2037],[163,0.2542],[180,0.2636],[198,0.2641],[210,0.2567],[235,0.1858],[252,0.075],[266,0]]
       }
       lipsTogether_3 {type=MovieSceneScriptingFloatChannel, default=0, const [0,30]}
       CTRL_R_mouth_lipD.X {
        type=MovieSceneScriptingFloatChannel
        default=0
        thinned_from=32
        keys [[0,-0.0582],[41,-0.0582],[46,-0.0584],[51,-0.0591],[60,-0.0616],[67,-0.0647],[75,-0.0743],[82,-0.0918],[87,-0.1163],[93,-0.1496],[99,-0.1629],[106,-0.169],[116,-0.1721],[121,-0.178],[130,-0.1993],[138,-0.2417],[142,-0.2732],[146,-0.2517],[155,-0.1641],[163,-0.1141],[180,-0.1086],[198,-0.1075],[210,-0.0878],[235,-0.0761],[252,-0.0653],[266,-0.0582]]
       }
       CTRL_R_mouth_lipD.Y {
        type=MovieSceneScriptingFloatChannel
        default=0
        thinned_from=32
        keys [[0,-0.1207],[31,-0.1207],[41,-0.1211],[46,-0.1232],[51,-0.1304],[56,-0.1417],[60,-0.1462],[67,-0.1305],[75,-0.0711],[82,0.0325],[87,0.1576],[93,0.3107],[99,0.3589],[106,0.3734],[112,0.3543],[116,0.317],[121,0.2927],[138,0.3444],[142,0.2869],[146,-0.153],[155,-0.4429],[163,-0.5269],[180,-0.5516],[198,-0.5267],[210,-0.418],[235,-0.2983],[252,-0.1897],[266,-0.1207]]
       }
       mouthPress_2_2 {
        type=MovieSceneScriptingFloatChannel
        default=0
        thinned_from=31
        keys [[0,0],[99,0.0001],[106,0.0002],[112,0.0006],[116,0.0024],[130,0.0079],[138,0.0188],[142,0.0337],[146,0.061],[155,0.0984],[163,0.1189],[180,0.1211],[198,0.1211],[210,0.116],[235,0.0855],[252,0.035],[266,0]]
       }
       lipsTogether_2_2 {type=MovieSceneScriptingFloatChannel, default=0, const [0,31]}
       CTRL_R_mouth_smileDepress.X {
        type=MovieSceneScriptingFloatChannel
        default=0
        thinned_from=30
        keys [[0,0],[31,0],[41,0.0002],[46,0.0014],[51,0.0059],[56,0.0134],[60,0.0186],[67,0.0242],[75,0.0385],[82,0.0652],[87,0.1085],[93,0.1722],[99,0.2115],[112,0.2103],[130,0.1875],[138,0.1645],[142,0.1398],[146,0.1093],[155,0.0785],[163,0.0555],[180,0.0095],[198,0.0075],[210,0.013],[220,0.0051],[235,0.0044],[252,0.0029],[266,0]]
       }
       CTRL_R_mouth_smileDepress.Y {
        type=MovieSceneScriptingFloatChannel
        default=0
        keys [[0,0.4374],[8,0.4469],[15,0.4726],[23,0.5181],[31,0.5557],[41,0.5711],[46,0.5701],[51,0.5581],[56,0.5408],[60,0.5259],[67,0.5245],[75,0.5551],[82,0.6203],[87,0.7213],[93,0.867],[99,0.9226],[112,0.9622],[130,0.8874],[138,0.8141],[142,0.7604],[146,0.7101],[155,0.6668],[163,0.6494],[180,0.7975],[198,0.81],[210,0.4955],[220,0.3543],[235,0.3711],[252,0.4162],[266,0.4374]]
       }
       CTRL_R_mouth_corner.X {
        type=MovieSceneScriptingFloatChannel
        default=0
        thinned_from=29
        keys [[0,0],[41,0],[46,-0.0001],[51,-0.0005],[56,-0.0012],[60,-0.0014],[67,-0.0001],[75,0.0048],[82,0.0132],[87,0.0237],[93,0.0369],[99,0.0446],[112,0.0441],[130,0.0393],[138,0.0344],[142,0.0293],[146,0.0228],[155,0.015],[163,0.0042],[180,-0.0012],[198,-0.0014],[210,0.0066],[235,0.0072],[252,0.003],[266,0]]
       }
       CTRL_R_mouth_corner.Y {
        type=MovieSceneScriptingFloatChannel
        default=0
        thinned_from=29
        keys [[0,0.395],[31,0.3951],[41,0.3957],[46,0.3995],[51,0.4116],[56,0.431],[60,0.4456],[67,0.4534],[75,0.4571],[82,0.4674],[87,0.5023],[93,0.4763],[99,0.3728],[112,0.3373],[130,0.3117],[138,0.2989],[142,0.2808],[146,0.2538],[155,0.1975],[163,0.1887],[180,0.2004],[198,0.2014],[210,0.1737],[235,0.2178],[252,0.3223],[266,0.395]]
       }
       lipsBlow_2 {type=MovieSceneScriptingFloatChannel, default=0, const [0,29]}
       mouthDimple_2 {
        type=MovieSceneScriptingFloatChannel
        default=0
        thinned_from=30
        keys [[0,0],[99,0.0001],[112,0.0013],[130,0.0107],[138,0.0462],[142,0.1254],[146,0.2407],[155,0.3517],[163,0.4129],[180,0.4212],[198,0.4229],[210,0.4216],[220,0.3943],[235,0.2986],[252,0.1222],[266,0]]
       }
       mouthStretch_2 {
        type=MovieSceneScriptingFloatChannel
        default=0
        thinned_from=30
        keys [[0,0],[67,0],[75,0.0003],[82,0.0034],[87,0.0177],[93,0.0445],[99,0.0563],[112,0.04],[130,0.0142],[138,0.0023],[142,0.0001],[266,0]]
       }
       mouthPurseU_2 {
        type=MovieSceneScriptingFloatChannel
        default=0
        thinned_from=30
        keys [[0,0],[67,0],[75,0.0009],[82,0.0103],[87,0.0537],[93,0.1353],[99,0.1711],[112,0.1215],[130,0.0431],[138,0.007],[142,0.0004],[146,0],[266,0]]
       }
       mouthPurseD_2 {
        type=MovieSceneScriptingFloatChannel
        default=0
        thinned_from=30
        keys [[0,0],[56,0],[60,0.0008],[67,0.006],[75,0.0237],[82,0.056],[87,0.1018],[93,0.1644],[99,0.201],[112,0.1951],[130,0.1652],[138,0.1351],[142,0.1023],[146,0.0616],[155,0.0233],[163,0.0052],[180,0.0036],[198,0.0032],[210,0.001],[220,0.0005],[266,0]]
       }
       mouthFunnelU_2 {
        type=MovieSceneScriptingFloatChannel
        default=0
        thinned_from=30
        keys [[0,0],[67,0],[75,0.0007],[82,0.0085],[87,0.0449],[93,0.1175],[99,0.1734],[112,0.2001],[130,0.2158],[138,0.2299],[142,0.2496],[146,0.2767],[155,0.303],[163,0.3144],[198,0.3143],[210,0.3048],[220,0.279],[235,0.21],[252,0.0858],[266,0]]
       }
       mouthFunnelD_2 {
        type=MovieSceneScriptingFloatChannel
        default=0
        thinned_from=30
        keys [[0,0],[67,0],[75,0.0003],[82,0.0031],[87,0.0164],[93,0.0414],[99,0.0523],[112,0.0372],[130,0.0132],[138,0.0021],[142,0.0001],[266,0]]
       }
       mouthCornerSharpnessU_2 {
        type=MovieSceneScriptingFloatChannel
        default=0
        thinned_from=30
        keys [[0,0],[56,-0.0001],[60,-0.0015],[67,-0.0118],[75,-0.0463],[82,-0.1112],[87,-0.2094],[93,-0.3484],[99,-0.3986],[112,-0.4199],[130,-0.3581],[138,-0.3018],[142,-0.2503],[146,-0.1919],[155,-0.1377],[163,-0.1088],[180,-0.1043],[198,-0.1034],[210,-0.0878],[220,-0.0784],[235,-0.0599],[252,-0.0246],[266,0]]
       }
       mouthCornerSharpnessD_2 {
        type=MovieSceneScriptingFloatChannel
        default=0
        thinned_from=30
        keys [[0,0],[56,-0.0001],[60,-0.0012],[67,-0.0092],[75,-0.0359],[82,-0.0833],[87,-0.146],[93,-0.2307],[99,-0.296],[112,-0.333],[130,-0.3474],[138,-0.3323],[142,-0.2836],[146,-0.2078],[155,-0.131],[163,-0.0903],[180,-0.085],[198,-0.0839],[210,-0.069],[220,-0.0612],[235,-0.0468],[252,-0.0192],[266,0]]
       }
       CTRL_L_mouth_suckBlow {
        type=MovieSceneScriptingFloatChannel
        default=0
        thinned_from=29
        keys [[0,-0.654],[56,-0.654],[60,-0.6552],[67,-0.6676],[75,-0.713],[82,-0.7942],[87,-0.9034],[93,-1.0507],[99,-1.1448],[112,-1.1646],[130,-1.1549],[138,-1.1426],[142,-1.1275],[146,-1.1064],[155,-1.0867],[163,-1.0934],[180,-1.1256],[198,-1.1178],[210,-1.0468],[235,-0.9277],[252,-0.7604],[266,-0.6726]]
       }
       CTRL_C_jaw.Location.X {type=MovieSceneScriptingFloatChannel, default=0, const [-0.0314,29]}
       CTRL_C_jaw.Location.Y {
        type=MovieSceneScriptingFloatChannel
        default=0
        thinned_from=29
        keys [[0,0],[67,0],[75,0.0021],[84,0.0067],[89,0.013],[95,0.0293],[101,0.0384],[106,0.0418],[114,0.0391],[130,0.0307],[138,0.0239],[148,0.0199],[157,0.0186],[198,0.016],[210,0.0151],[235,0.0117],[252,0.0038],[266,0]]
       }
       CTRL_C_jaw.Location.Z {
        type=MovieSceneScriptingFloatChannel
        default=0
        thinned_from=29
        keys [[0,0.7168],[67,0.7168],[75,0.7183],[84,0.7216],[89,0.7249],[95,0.7345],[101,0.7374],[114,0.7393],[130,0.7369],[138,0.7318],[157,0.7241],[235,0.7212],[252,0.7171],[266,0.7168]]
       }
       jawOpenExtreme {type=MovieSceneScriptingFloatChannel, default=0, const [0,29]}
       chinRaiseUL {type=MovieSceneScriptingFloatChannel, default=0, const [0,29]}
       chinRaiseUR {type=MovieSceneScriptingFloatChannel, default=0, const [0,29]}
       chinRaiseDL {type=MovieSceneScriptingFloatChannel, default=0, const [0,29]}
       chinRaiseDR {type=MovieSceneScriptingFloatChannel, default=0, const [0,29]}
       CTRL_R_brow.X {
        type=MovieSceneScriptingFloatChannel
        default=0
        thinned_from=36
        keys [[0,0],[41,0.0001],[46,0.0011],[51,0.0069],[56,0.0205],[60,0.0282],[67,0.0278],[75,0.0189],[80,0.0141],[84,0.0517],[90,0.1318],[96,0.0556],[102,0.012],[114,0.0033],[130,0.0016],[138,0.0011],[165,0.0006],[181,0.0033],[197,0.041],[200,0.1976],[203,0.3432],[206,0.1585],[210,0.0279],[223,0.0017],[229,0.0002],[236,0.0028],[239,0.0095],[242,0.0057],[245,0.0012],[266,0]]
       }
       CTRL_R_brow.Y {
        type=MovieSceneScriptingFloatChannel
        default=0
        keys [[0,-0.0367],[4,-0.0264],[8,0.0121],[15,0.0805],[27,0.1175],[41,0.1269],[46,0.1137],[51,0.0401],[56,-0.1075],[60,-0.1901],[67,-0.2069],[75,-0.1864],[80,-0.1461],[84,-0.0891],[90,0.023],[96,0.1394],[102,0.1944],[114,0.2091],[130,0.2123],[138,0.2103],[148,0.1998],[157,0.1854],[165,0.1779],[181,0.1717],[197,0.1586],[200,0.126],[203,-0.0132],[206,0.1565],[210,0.2027],[223,0.2189],[229,0.2025],[236,0.1549],[239,0.0778],[242,0.0068],[245,-0.0201],[266,-0.0367]]
       }
       CTRL_R_brow_inner {
        type=MovieSceneScriptingFloatChannel
        default=0
        thinned_from=34
        keys [[0,-0.0367],[4,-0.0264],[15,0.079],[27,0.1106],[41,0.1169],[46,0.1129],[51,0.0917],[56,0.0544],[60,0.035],[67,0.0295],[75,0.0446],[84,0.1093],[90,0.2034],[96,0.265],[102,0.2626],[114,0.2543],[130,0.2506],[138,0.2458],[148,0.2316],[157,0.2154],[165,0.2073],[181,0.2003],[197,0.1821],[200,0.1232],[203,0.0839],[206,0.0994],[210,0.1086],[229,0.093],[236,0.057],[239,-0.0051],[242,-0.0293],[245,-0.031],[266,-0.0367]]
       }
       CTRL_R_brow_outer {
        type=MovieSceneScriptingFloatChannel
        default=0
        keys [[0,-0.0367],[4,-0.0264],[8,0.0122],[15,0.0819],[27,0.1222],[41,0.1304],[46,0.1178],[51,0.0529],[56,-0.0708],[60,-0.1394],[67,-0.1515],[75,-0.1215],[84,-0.0523],[90,0.0524],[96,0.2774],[102,0.3916],[114,0.4329],[130,0.4409],[138,0.44],[148,0.4355],[157,0.4301],[165,0.4267],[181,0.4172],[197,0.3831],[200,0.3232],[203,0.2679],[206,0.2516],[210,0.2417],[229,0.2086],[236,0.1559],[239,0.0777],[242,0.0062],[245,-0.0207],[266,-0.0367]]
       }
       CTRL_R_eye_blink {
        type=MovieSceneScriptingFloatChannel
        default=0
        thinned_from=51
        keys [[0,0],[8,0],[11,0.0379],[13,0.3716],[14,0.8913],[15,0.941],[20,1],[21,0.7012],[24,0.273],[28,0.0644],[31,0],[51,0],[54,0.0681],[56,0.8538],[58,1.0369],[59,0.6509],[61,0.0986],[63,0.0384],[67,0],[84,0],[89,-0.0143],[95,-1.4325],[101,-1.6778],[114,-1.7858],[130,-1.8189],[138,-1.8529],[148,-1.9143],[157,-1],[165,-1.9739],[178,-1.8821],[199,-0.7668],[200,0.3661],[201,1],[204,1],[205,0.799],[208,0.1723],[211,0],[235,0],[236,0.0774],[237,0.392],[238,0.9308],[239,3.5869],[240,0.4155],[241,0.1963],[243,0.058],[245,0],[266,0]]
       }
       CTRL_R_eye_eyelidU.X {type=MovieSceneScriptingFloatChannel, default=0, const [0,28]}
       CTRL_R_eye_eyelidU.Y {
        type=MovieSceneScriptingFloatChannel
        default=0
        thinned_from=28
        keys [[0,0],[67,0],[75,0.4559],[84,1],[89,0.7789],[95,0.8328],[101,1],[114,0.9841],[130,0.9572],[138,0.9594],[157,1],[165,1],[178,0.9642],[198,0.8228],[210,0.7093],[235,0.6522],[252,0.0439],[266,0]]
       }
       CTRL_R_eye_eyelidD.X {type=MovieSceneScriptingFloatChannel, default=0, const [0,28]}
       CTRL_R_eye_eyelidD.Y {
        type=MovieSceneScriptingFloatChannel
        default=0
        thinned_from=28
        keys [[0,0],[8,0.0031],[15,0.0115],[23,0.1579],[31,0.1629],[41,0.1649],[46,0.1644],[51,0.1549],[56,0.1028],[60,0.0941],[67,0.0779],[75,0.0408],[84,-0.0349],[89,-0.0694],[95,-0.0819],[101,-0.0896],[114,-0.0914],[138,-0.092],[148,-0.0924],[157,-0.0942],[165,-0.0934],[210,-0.0958],[235,-0.0737],[252,-0.0214],[266,0]]
       }
       CTRL_R_eye_squintInner {
        type=MovieSceneScriptingFloatChannel
        default=0
        thinned_from=28
        keys [[0,0.187],[67,0.187],[75,0.2251],[84,0.3535],[89,0.4067],[95,0.8464],[101,0.9875],[114,0.973],[130,0.9486],[138,0.8995],[148,0.7072],[157,0.3034],[165,0.2337],[180,0.2054],[198,0.2067],[210,0.2064],[235,0.1986],[252,0.1937],[266,0.187]]
       }
       CTRL_R_eye_cheekRaise {
        type=MovieSceneScriptingFloatChannel
        default=0
        thinned_from=28
        keys [[0,0],[46,0],[60,0.0006],[67,0.0006],[75,0.0217],[84,0.1241],[89,0.2598],[95,0.4567],[101,0.7078],[114,0.8237],[130,0.8561],[138,0.8529],[165,0.8554],[198,0.8429],[210,0.8195],[235,0.584],[252,0.0418],[266,0]]
       }
       CTRL_R_mouth_suckBlow {
        type=MovieSceneScriptingFloatChannel
        default=0
        keys [[0,-0.6462],[8,-0.6587],[15,-0.7164],[23,-0.8211],[31,-0.8764],[41,-0.8863],[46,-0.8805],[51,-0.8549],[56,-0.8129],[60,-0.7883],[67,-0.7849],[75,-0.7977],[82,-0.8301],[87,-0.9078],[93,-1.0428],[99,-1.1283],[112,-1.1251],[130,-1.077],[138,-1.0305],[142,-0.9811],[146,-0.9182],[155,-0.8663],[163,-0.8955],[180,-0.9988],[198,-1.0034],[210,-0.8814],[235,-0.7684],[252,-0.6869],[266,-0.6515]]
       }
       CTRL_C_eye.X {
        type=MovieSceneScriptingFloatChannel
        default=0
        thinned_from=71
        keys [[0,0.4247],[3,0.4247],[4,0.3944],[8,0.3931],[9,0.4575],[16,0.4624],[18,0.7929],[19,0.914],[23,0.9424],[31,0.9588],[41,0.9652],[42,0.9575],[43,0.9549],[52,0.9536],[53,0.8932],[56,0.8582],[57,0.3056],[58,0.132],[60,0.0903],[65,0.0855],[66,0.211],[67,0.2132],[73,0.2172],[74,0.2267],[76,0.2292],[84,0.2778],[89,0.2926],[95,0.2732],[101,0.2668],[105,0.2634],[106,0.2436],[109,0.2396],[114,0.2365],[124,0.3379],[125,0.3506],[130,0.3527],[138,0.3159],[146,0.2557],[147,0.2542],[151,0.2911],[152,0.3316],[157,0.3353],[165,0.3481],[177,0.4423],[178,0.4726],[186,0.4825],[191,0.4842],[192,0.4892],[199,0.4895],[200,0.3522],[203,0.2778],[204,-0.177],[206,-0.2717],[210,-0.3007],[224,-0.3098],[225,-0.2434],[228,-0.2277],[232,-0.2458],[237,-0.3351],[238,0.3659],[239,0.4854],[240,0.7305],[252,0.453],[256,0.4184],[257,0.3679],[260,0.3628],[262,0.3628],[263,0.4135],[266,0.4247]]
       }
       CTRL_C_eye.Y {
        type=MovieSceneScriptingFloatChannel
        default=0
        keys [[0,0.2434],[3,0.2434],[4,0.2786],[8,0.2801],[9,0.2387],[16,0.238],[18,-0.0164],[19,-0.034],[23,-0.0381],[31,-0.0405],[41,-0.0414],[42,-0.1093],[43,-0.1288],[46,-0.1349],[52,-0.1349],[53,-0.1243],[56,-0.1181],[57,-0.0399],[58,-0.0182],[60,-0.0096],[65,0.0022],[66,0.0613],[67,0.0761],[73,0.0649],[74,-0.0085],[76,-0.0329],[84,-0.1101],[89,-0.1806],[95,-0.552],[101,-0.6166],[105,-0.6128],[106,-0.5528],[109,-0.5377],[114,-0.5043],[124,-0.3849],[125,-0.269],[130,-0.2264],[138,-0.2106],[146,0.1436],[147,0.1638],[151,0.2021],[152,0.3858],[157,0.4087],[165,0.434],[177,-0.371],[178,-0.3346],[186,-0.3168],[191,-0.3089],[192,-0.2249],[199,-0.2031],[200,-0.2571],[201,-0.3221],[203,-0.5813],[204,-0.7991],[206,0.2613],[210,0.3607],[224,0.3844],[225,0.4615],[228,0.4761],[232,0.4338],[237,0.4338],[238,0.1904],[239,0.2594],[240,0.3205],[252,0.2083],[256,0.1792],[257,0.2949],[260,0.3033],[262,0.3033],[263,0.2539],[266,0.2434]]
       }
       CTRL_L_eye.X {type=MovieSceneScriptingFloatChannel, default=0, const [0,28]}
       CTRL_L_eye.Y {type=MovieSceneScriptingFloatChannel, default=0, const [0,28]}
       CTRL_R_eye.X {type=MovieSceneScriptingFloatChannel, default=0, const [0,28]}
       CTRL_R_eye.Y {type=MovieSceneScriptingFloatChannel, default=0, const [0,28]}
       eyesConvergence {type=MovieSceneScriptingFloatChannel, default=0, const [0,68]}
       eyelidsReduce {
        type=MovieSceneScriptingFloatChannel
        default=0
        thinned_from=68
        keys [[0,0],[66,0],[73,0.0032],[74,0.0062],[76,0.0216],[84,0.114],[89,0.1443],[95,0.3304],[101,0.3882],[105,0.3876],[106,0.3776],[109,0.3751],[114,0.3693],[125,0.3103],[130,0.3029],[138,0.2776],[147,0.1864],[151,0.1753],[152,0.1215],[157,0.1146],[165,0.104],[177,0.0962],[178,0.0959],[186,0.0958],[191,0.0953],[192,0.0881],[199,0.0872],[200,0.086],[203,0.0855],[204,0.0843],[206,0.0814],[210,0.081],[224,0.0787],[225,0.0776],[232,0.0771],[237,0.0771],[238,0.0419],[239,0.0082],[256,0.0044],[257,0.0028],[262,0.0027],[263,0.0004],[266,0]]
       }
       lookAtSwitch {type=MovieSceneScriptingFloatChannel, default=0, const [0,28]}
       CTRL_C_tongue.Location.X {type=MovieSceneScriptingFloatChannel, default=0, const [0,28]}
       CTRL_C_tongue.Location.Y {
        type=MovieSceneScriptingFloatChannel
        default=0
        thinned_from=28
        keys [[0,0],[67,0],[75,-0.0974],[84,-0.2711],[89,-0.5013],[95,-0.8101],[101,-0.9932],[114,-0.9789],[130,-0.8593],[138,-0.7123],[148,-0.5208],[157,-0.3559],[165,-0.279],[178,-0.2476],[198,-0.2253],[210,-0.2009],[235,-0.1357],[252,-0.0119],[266,0]]
       }
       CTRL_C_tongue.Location.Z {type=MovieSceneScriptingFloatChannel, default=0, const [0,28]}
       wideNarrow {type=MovieSceneScriptingFloatChannel, default=0, const [0,28]}
       rollUpDown {type=MovieSceneScriptingFloatChannel, default=0, const [0,28]}
       rollLeftRight {type=MovieSceneScriptingFloatChannel, default=0, const [0,28]}
       CTRL_C_teethU.Location.X {type=MovieSceneScriptingFloatChannel, default=0, const [0,28]}
       CTRL_C_teethU.Location.Y {type=MovieSceneScriptingFloatChannel, default=0, const [0.0858,28]}
       CTRL_C_teethU.Location.Z {type=MovieSceneScriptingFloatChannel, default=0, const [0,28]}
       rotateFwdBack {type=MovieSceneScriptingFloatChannel, default=0, const [0,28]}
       CTRL_C_teethD.Location.X {type=MovieSceneScriptingFloatChannel, default=0, const [0,28]}
       CTRL_C_teethD.Location.Y {
        type=MovieSceneScriptingFloatChannel
        default=0
        thinned_from=28
        keys [[0,0],[67,0],[75,-0.0051],[84,-0.1275],[89,-0.3438],[95,-0.385],[101,-0.3939],[114,-0.3863],[130,-0.386],[138,-0.3774],[148,-0.3283],[157,-0.1499],[165,-0.0432],[178,-0.0189],[198,-0.0172],[210,-0.016],[235,-0.0152],[252,-0.001],[266,0]]
       }
       CTRL_C_teethD.Location.Z {type=MovieSceneScriptingFloatChannel, default=0, thinned_from=28, keys [[0,0],[266,0]]}
       rotateFwdBack_2 {type=MovieSceneScriptingFloatChannel, default=0, const [0,28]}
      ]
     }]
   }
   {
    type=MovieSceneControlRigParameterTrack
    display_name=MR_Biped_v01
    sections [{type=MovieSceneControlRigParameterSection}, {type=MovieSceneControlRigParameterSection}, {type=MovieSceneControlRigParameterSection}]
   }
  ]
 }
 Head {
  id=#2
  parent_id=#1
  possessed_class=SkeletalMeshComponentBudgeted
  tracks [{
    type=MovieSceneControlRigParameterTrack
    display_name=Fortnite_Face_BoneTweaker_CtrlRig
    sections [{type=MovieSceneControlRigParameterSection}]
   }]
 }
]
display_rate {denominator=1, numerator=30}
tick_resolution {denominator=1, numerator=24000}
tracks []
